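# Simple unlit material for Athena skydive ring quest interactables, with a glow color and a wobble term.
# The color chain multiplies the ColorInner parameter by OverallBrightness.
# A separate chain scales the vector from world position to object pivot per axis by WobbleScale, then by a time-driven sine, giving an oscillation toward and away from the pivot.
# The file does not show which material outputs these two chains feed.
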Material /Game/Athena/Items/QuestInteractables/SkydiveRings/FX/M_Athena_SkydiveRings_Simple
graph {
 properties {pixelDepthOffsetMode=PDOM_Legacy, shadingModel=MSM_Unlit, usedShadingModels=Unlit}
 nodes [
  Multiply(
   A=VectorParameter() {parameterName=ColorInner, defaultValue=<rgba 0.0012 0.4397 0.807 1>}.RGB
   B=ScalarParameter() {defaultValue=8, parameterName=OverallBrightness}
  )
  n1 = MaterialFunctionCall() {materialFunction=&ObjectPivotPoint}
  n2 = WorldPosition()
  Multiply(
   A=Multiply(A=Subtract(A=@n1.Object Pivot Location, B=@n2.XYZ), B=VectorParameter() {parameterName=WobbleScale, channelNames {a="", b=Z, g=Y, r=X}})
   B=Sine(None=Time()) {period=0.95}
  )
 ]
}
